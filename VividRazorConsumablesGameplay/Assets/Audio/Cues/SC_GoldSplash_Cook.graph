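# Sound cue for the Gold Splash consumable's cook sound, played as player foley with generic equip attenuation.
# A quality-level switch sends its first two slots to a random pick of six wave variations and its last two to a reduced set of three.
# A modulator on top adds slight random pitch variation at constant volume.
# Its own concurrency override blocks new instances once the limit is reached.
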SoundCue /VividRazorConsumablesGameplay/Assets/Audio/Cues/SC_GoldSplash_Cook
root=Modulator_12
nodes [
 {id=Modulator_12, children [QualityLevel_0], properties {pitchMin=0.9, volumeMax=1, volumeMin=1}}
 {id=QualityLevel_0, children [Random_1,Random_1,Random_2,Random_2]}
 {
  id=Random_1
  children [WavePlayer_0,WavePlayer_5,WavePlayer_6,WavePlayer_7,WavePlayer_8,WavePlayer_9]
  properties {bShouldExcludeFromBranchCulling=true, weights [1,1,1,1,1,1]}
 }
 {id=Random_2, children [WavePlayer_10, WavePlayer_11, WavePlayer_12], properties {weights [1,1,1]}}
 {id=WavePlayer_0, properties {soundWaveAssetPtr=/VividRazorConsumablesGameplay/Assets/Audio/Waves/GoldSplash_Cook_01}}
 {id=WavePlayer_10, properties {soundWaveAssetPtr=&GoldSplash_Cook_01}}
 {id=WavePlayer_11, properties {soundWaveAssetPtr=/VividRazorConsumablesGameplay/Assets/Audio/Waves/GoldSplash_Cook_04}}
 {id=WavePlayer_12, properties {soundWaveAssetPtr=/VividRazorConsumablesGameplay/Assets/Audio/Waves/GoldSplash_Cook_06}}
 {id=WavePlayer_5, properties {soundWaveAssetPtr=/VividRazorConsumablesGameplay/Assets/Audio/Waves/GoldSplash_Cook_02}}
 {id=WavePlayer_6, properties {soundWaveAssetPtr=/VividRazorConsumablesGameplay/Assets/Audio/Waves/GoldSplash_Cook_03}}
 {id=WavePlayer_7, properties {soundWaveAssetPtr=&GoldSplash_Cook_04}}
 {id=WavePlayer_8, properties {soundWaveAssetPtr=/VividRazorConsumablesGameplay/Assets/Audio/Waves/GoldSplash_Cook_05}}
 {id=WavePlayer_9, properties {soundWaveAssetPtr=&GoldSplash_Cook_06}}
]
properties {
 attenuationSettings=/Game/Sounds/Attenuation/Fort_Player_Equip_Generic_Att
 bOverrideConcurrency=true
 maxDistance=1800
 soundClassObject=&Fort_Player_Foley
 volumeMultiplier=1
 attenuationOverrides {falloffDistance=2400, attenuationShapeExtents {x=600}}
 concurrencyOverrides {resolutionRule=PreventNew}
}
