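# Sound cue for the Gold Splash consumable's cook-stop sound, classed as player foley with generic equip attenuation.
# A quality-level node sends the first two levels to a random pick of six wave variations and the last two to a smaller pick of the first three.
# A modulator above it varies pitch while holding volume steady.
# The cue lowers overall pitch and volume, and its concurrency override blocks new instances once the limit is reached.
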SoundCue /VividRazorConsumablesGameplay/Assets/Audio/Cues/SC_GoldSplash_CookStop
root=Modulator_12
nodes [
 {id=Modulator_12, children [QualityLevel_0], properties {pitchMin=0.9, volumeMax=1, volumeMin=1}}
 {id=QualityLevel_0, children [Random_0,Random_0,Random_3,Random_3]}
 {
  id=Random_0
  children [WavePlayer_1, WavePlayer_2, WavePlayer_3, WavePlayer_4, WavePlayer_13, WavePlayer_14]
  properties {bShouldExcludeFromBranchCulling=true, weights [1,1,1,1,1,1]}
 }
 {id=Random_3, children [WavePlayer_15, WavePlayer_16, WavePlayer_17], properties {weights [1,1,1]}}
 {id=WavePlayer_1, properties {soundWaveAssetPtr=/VividRazorConsumablesGameplay/Assets/Audio/Waves/GoldSplash_CookStop_01}}
 {id=WavePlayer_13, properties {soundWaveAssetPtr=/VividRazorConsumablesGameplay/Assets/Audio/Waves/GoldSplash_CookStop_05}}
 {id=WavePlayer_14, properties {soundWaveAssetPtr=/VividRazorConsumablesGameplay/Assets/Audio/Waves/GoldSplash_CookStop_06}}
 {id=WavePlayer_15, properties {soundWaveAssetPtr=&GoldSplash_CookStop_01}}
 {id=WavePlayer_16, properties {soundWaveAssetPtr=/VividRazorConsumablesGameplay/Assets/Audio/Waves/GoldSplash_CookStop_02}}
 {id=WavePlayer_17, properties {soundWaveAssetPtr=/VividRazorConsumablesGameplay/Assets/Audio/Waves/GoldSplash_CookStop_03}}
 {id=WavePlayer_2, properties {soundWaveAssetPtr=&GoldSplash_CookStop_02}}
 {id=WavePlayer_3, properties {soundWaveAssetPtr=&GoldSplash_CookStop_03}}
 {id=WavePlayer_4, properties {soundWaveAssetPtr=/VividRazorConsumablesGameplay/Assets/Audio/Waves/GoldSplash_CookStop_04}}
]
properties {
 attenuationSettings=/Game/Sounds/Attenuation/Fort_Player_Equip_Generic_Att
 bOverrideConcurrency=true
 maxDistance=1800
 pitchMultiplier=0.8
 soundClassObject=&Fort_Player_Foley
 volumeMultiplier=0.6
 attenuationOverrides {falloffDistance=2400, attenuationShapeExtents {x=600}}
 concurrencyOverrides {resolutionRule=PreventNew}
}
