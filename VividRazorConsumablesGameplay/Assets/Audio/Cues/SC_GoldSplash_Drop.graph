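# Sound cue for dropping the Gold Splash consumable, using weapon-drop attenuation and the player foley sound class.
# A quality-level switch feeds the first two levels from a random pool of six drop waves and the last two from a three-wave subset.
# A root modulator varies pitch while holding volume fixed.
# Its concurrency override blocks new instances once the limit is reached.
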SoundCue /VividRazorConsumablesGameplay/Assets/Audio/Cues/SC_GoldSplash_Drop
root=Modulator_12
nodes [
 {id=Modulator_12, children [QualityLevel_0], properties {pitchMin=0.9, volumeMax=1, volumeMin=1}}
 {id=QualityLevel_0, children [Random_2,Random_2,Random_3,Random_3]}
 {
  id=Random_2
  children [WavePlayer_16, WavePlayer_17, WavePlayer_18, WavePlayer_19, WavePlayer_20, WavePlayer_21]
  properties {bShouldExcludeFromBranchCulling=true, weights [1,1,1,1,1,1]}
 }
 {id=Random_3, children [WavePlayer_24, WavePlayer_23, WavePlayer_22], properties {weights [1,1,1]}}
 {id=WavePlayer_16, properties {soundWaveAssetPtr=/VividRazorConsumablesGameplay/Assets/Audio/Waves/GoldSplash_Drop_06}}
 {id=WavePlayer_17, properties {soundWaveAssetPtr=/VividRazorConsumablesGameplay/Assets/Audio/Waves/GoldSplash_Drop_05}}
 {id=WavePlayer_18, properties {soundWaveAssetPtr=/VividRazorConsumablesGameplay/Assets/Audio/Waves/GoldSplash_Drop_04}}
 {id=WavePlayer_19, properties {soundWaveAssetPtr=/VividRazorConsumablesGameplay/Assets/Audio/Waves/GoldSplash_Drop_03}}
 {id=WavePlayer_20, properties {soundWaveAssetPtr=/VividRazorConsumablesGameplay/Assets/Audio/Waves/GoldSplash_Drop_02}}
 {id=WavePlayer_21, properties {soundWaveAssetPtr=/VividRazorConsumablesGameplay/Assets/Audio/Waves/GoldSplash_Drop_01}}
 {id=WavePlayer_22, properties {soundWaveAssetPtr=&GoldSplash_Drop_01}}
 {id=WavePlayer_23, properties {soundWaveAssetPtr=&GoldSplash_Drop_03}}
 {id=WavePlayer_24, properties {soundWaveAssetPtr=&GoldSplash_Drop_06}}
]
properties {
 attenuationSettings=&Fort_Weapon_Drop_Att
 bOverrideConcurrency=true
 maxDistance=3000
 soundClassObject=&Fort_Player_Foley
 volumeMultiplier=1.3
 attenuationOverrides {falloffDistance=2400, attenuationShapeExtents {x=600}}
 concurrencyOverrides {resolutionRule=PreventNew}
}
