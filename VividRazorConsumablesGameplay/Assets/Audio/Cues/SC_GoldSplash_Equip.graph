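# Sound cue for the equip sound of the GoldSplash consumable, classed as player foley.
# A Random node picks among four equally weighted equip waves and feeds a root Modulator that varies pitch at constant volume.
# The cue uses the shared generic player-equip attenuation asset, boosts volume, and overrides concurrency to prevent new instances once the limit is reached.
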SoundCue /VividRazorConsumablesGameplay/Assets/Audio/Cues/SC_GoldSplash_Equip
root=Modulator_12
nodes [
 {id=Modulator_12, children [Random_0], properties {pitchMin=0.9, volumeMax=1, volumeMin=1}}
 {
  id=Random_0
  children [WavePlayer_1,WavePlayer_2,WavePlayer_3,WavePlayer_4]
  properties {bShouldExcludeFromBranchCulling=true, weights [1,1,1,1]}
 }
 {id=WavePlayer_1, properties {soundWaveAssetPtr=/VividRazorConsumablesGameplay/Assets/Audio/Waves/GoldSplash_Equip_01}}
 {id=WavePlayer_2, properties {soundWaveAssetPtr=/VividRazorConsumablesGameplay/Assets/Audio/Waves/GoldSplash_Equip_03}}
 {id=WavePlayer_3, properties {soundWaveAssetPtr=/VividRazorConsumablesGameplay/Assets/Audio/Waves/GoldSplash_Equip_04}}
 {id=WavePlayer_4, properties {soundWaveAssetPtr=/VividRazorConsumablesGameplay/Assets/Audio/Waves/GoldSplash_Equip_05}}
]
properties {
 attenuationSettings=/Game/Sounds/Attenuation/Fort_Player_Equip_Generic_Att
 bOverrideConcurrency=true
 maxDistance=1800
 soundClassObject=&Fort_Player_Foley
 volumeMultiplier=2
 attenuationOverrides {falloffDistance=2400, attenuationShapeExtents {x=600}}
 concurrencyOverrides {resolutionRule=PreventNew}
}
