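# Looping Sound Cue for the GoldSplash consumable's in-air sound, in the gadget items sound class.
# A quality switch plays an equally weighted random pick of three loop waves on the first two levels, and only the first wave on the other two.
# A modulator on top varies pitch at fixed volume.
# It uses the smoke grenade attenuation asset and a concurrency override that prevents new instances at the limit.
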SoundCue /VividRazorConsumablesGameplay/Assets/Audio/Cues/SC_GoldSplash_InAir_Loop
root=Modulator_12
nodes [
 {id=Modulator_12, children [QualityLevel_1], properties {pitchMin=0.9, volumeMax=1, volumeMin=1}}
 {id=QualityLevel_1, children [Random_0,Random_0,WavePlayer_4,WavePlayer_4]}
 {id=Random_0, children [WavePlayer_1,WavePlayer_2,WavePlayer_3], properties {bShouldExcludeFromBranchCulling=true, weights [1,1,1]}}
 {id=WavePlayer_1, properties {bLooping=true, soundWaveAssetPtr=/VividRazorConsumablesGameplay/Assets/Audio/Waves/GoldSplash_InAir_Loop_01}}
 {id=WavePlayer_2, properties {bLooping=true, soundWaveAssetPtr=/VividRazorConsumablesGameplay/Assets/Audio/Waves/GoldSplash_InAir_Loop_02}}
 {id=WavePlayer_3, properties {bLooping=true, soundWaveAssetPtr=/VividRazorConsumablesGameplay/Assets/Audio/Waves/GoldSplash_InAir_Loop_03}}
 {id=WavePlayer_4, properties {bLooping=true, soundWaveAssetPtr=&GoldSplash_InAir_Loop_01}}
]
properties {
 attenuationSettings=/Game/Sounds/Attenuation/Fort_Grenade_Smoke_Att
 bOverrideConcurrency=true
 maxDistance=6800
 soundClassObject=&Fort_Gadgets_Items
 volumeMultiplier=0.7
 attenuationOverrides {falloffDistance=2400, attenuationShapeExtents {x=600}}
 concurrencyOverrides {resolutionRule=PreventNew}
}
