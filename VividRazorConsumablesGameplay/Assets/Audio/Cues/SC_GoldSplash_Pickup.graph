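# Sound cue for the GoldSplash consumable pickup sound, classed as player foley.
# A Random node picks among five equally weighted wave variations, then a Modulator varies pitch while leaving volume fixed.
# The cue boosts its volume and uses the shared weapon pickup attenuation asset for spatial falloff.
# Its own concurrency override prevents new instances from starting once the voice limit is reached.
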SoundCue /VividRazorConsumablesGameplay/Assets/Audio/Cues/SC_GoldSplash_Pickup
root=Modulator_12
nodes [
 {id=Modulator_12, children [Random_1], properties {pitchMin=0.9, volumeMax=1, volumeMin=1}}
 {
  id=Random_1
  children [WavePlayer_0,WavePlayer_5,WavePlayer_6,WavePlayer_7,WavePlayer_8]
  properties {bShouldExcludeFromBranchCulling=true, weights [1,1,1,1,1]}
 }
 {id=WavePlayer_0, properties {soundWaveAssetPtr=/VividRazorConsumablesGameplay/Assets/Audio/Waves/GoldSplash_Pickup_01}}
 {id=WavePlayer_5, properties {soundWaveAssetPtr=/VividRazorConsumablesGameplay/Assets/Audio/Waves/GoldSplash_Pickup_02}}
 {id=WavePlayer_6, properties {soundWaveAssetPtr=/VividRazorConsumablesGameplay/Assets/Audio/Waves/GoldSplash_Pickup_03}}
 {id=WavePlayer_7, properties {soundWaveAssetPtr=/VividRazorConsumablesGameplay/Assets/Audio/Waves/GoldSplash_Pickup_04}}
 {id=WavePlayer_8, properties {soundWaveAssetPtr=/VividRazorConsumablesGameplay/Assets/Audio/Waves/GoldSplash_Pickup_05}}
]
properties {
 attenuationSettings=&Fort_Weapon_Pickup_Att
 bOverrideConcurrency=true
 maxDistance=3500
 soundClassObject=&Fort_Player_Foley
 volumeMultiplier=2
 attenuationOverrides {falloffDistance=2400, attenuationShapeExtents {x=600}}
 concurrencyOverrides {resolutionRule=PreventNew}
}
