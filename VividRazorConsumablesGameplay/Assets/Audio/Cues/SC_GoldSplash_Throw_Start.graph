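# Sound cue for the start of a GoldSplash consumable throw, routed to the gadgets/items sound class with smoke grenade attenuation.
# A quality-level node sends two tiers to a seven-variant random pool and the other two to a smaller pool reusing three of those waves.
# A root modulator varies pitch while holding volume steady, and overridden concurrency blocks new instances once the limit is reached.
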SoundCue /VividRazorConsumablesGameplay/Assets/Audio/Cues/SC_GoldSplash_Throw_Start
root=Modulator_12
nodes [
 {id=Modulator_12, children [QualityLevel_0], properties {pitchMin=0.9, volumeMax=1, volumeMin=1}}
 {id=QualityLevel_0, children [Random_2,Random_2,Random_3,Random_3]}
 {
  id=Random_2
  children [WavePlayer_0, WavePlayer_12, WavePlayer_13, WavePlayer_14, WavePlayer_15, WavePlayer_16, WavePlayer_17]
  properties {bShouldExcludeFromBranchCulling=true, weights [1,1,1,1,1,1,1]}
 }
 {id=Random_3, children [WavePlayer_18, WavePlayer_19, WavePlayer_20], properties {weights [1,1,1]}}
 {id=WavePlayer_0, properties {soundWaveAssetPtr=/VividRazorConsumablesGameplay/Assets/Audio/Waves/GoldSplash_Throw_Start_07}}
 {id=WavePlayer_12, properties {soundWaveAssetPtr=/VividRazorConsumablesGameplay/Assets/Audio/Waves/GoldSplash_Throw_Start_06}}
 {id=WavePlayer_13, properties {soundWaveAssetPtr=/VividRazorConsumablesGameplay/Assets/Audio/Waves/GoldSplash_Throw_Start_05}}
 {id=WavePlayer_14, properties {soundWaveAssetPtr=/VividRazorConsumablesGameplay/Assets/Audio/Waves/GoldSplash_Throw_Start_04}}
 {id=WavePlayer_15, properties {soundWaveAssetPtr=/VividRazorConsumablesGameplay/Assets/Audio/Waves/GoldSplash_Throw_Start_03}}
 {id=WavePlayer_16, properties {soundWaveAssetPtr=/VividRazorConsumablesGameplay/Assets/Audio/Waves/GoldSplash_Throw_Start_02}}
 {id=WavePlayer_17, properties {soundWaveAssetPtr=/VividRazorConsumablesGameplay/Assets/Audio/Waves/GoldSplash_Throw_Start_01}}
 {id=WavePlayer_18, properties {soundWaveAssetPtr=&GoldSplash_Throw_Start_03}}
 {id=WavePlayer_19, properties {soundWaveAssetPtr=&GoldSplash_Throw_Start_02}}
 {id=WavePlayer_20, properties {soundWaveAssetPtr=&GoldSplash_Throw_Start_01}}
]
properties {
 attenuationSettings=/Game/Sounds/Attenuation/Fort_Grenade_Smoke_Att
 bOverrideConcurrency=true
 maxDistance=6800
 soundClassObject=&Fort_Gadgets_Items
 volumeMultiplier=0.7
 attenuationOverrides {falloffDistance=2400, attenuationShapeExtents {x=600}}
 concurrencyOverrides {resolutionRule=PreventNew}
}
