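# Sound cue for the landing sound of a thrown Gold Splash consumable, classed under gadget and item sounds.
# A quality-level node sends its first two slots to a random pick of eight landing waves and its last two to a random pick of only the first three.
# A modulator varies pitch at constant volume.
# It uses the smoke grenade attenuation, and its concurrency override prevents new instances at the limit.
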SoundCue /VividRazorConsumablesGameplay/Assets/Audio/Cues/SC_GoldSplash_Thrown_Land
root=Modulator_12
nodes [
 {id=Modulator_12, children [QualityLevel_0], properties {pitchMin=0.9, volumeMax=1, volumeMin=1}}
 {id=QualityLevel_0, children [Random_0,Random_0,Random_1,Random_1]}
 {
  id=Random_0
  children [WavePlayer_1,WavePlayer_2,WavePlayer_3,WavePlayer_4,WavePlayer_5,WavePlayer_6,WavePlayer_7,WavePlayer_8]
  properties {bShouldExcludeFromBranchCulling=true, weights [1,1,1,1,1,1,1,1]}
 }
 {id=Random_1, children [WavePlayer_9, WavePlayer_10, WavePlayer_11], properties {weights [1,1,1]}}
 {id=WavePlayer_1, properties {soundWaveAssetPtr=/VividRazorConsumablesGameplay/Assets/Audio/Waves/GoldSplash_Thrown_Land_01}}
 {id=WavePlayer_10, properties {soundWaveAssetPtr=/VividRazorConsumablesGameplay/Assets/Audio/Waves/GoldSplash_Thrown_Land_02}}
 {id=WavePlayer_11, properties {soundWaveAssetPtr=/VividRazorConsumablesGameplay/Assets/Audio/Waves/GoldSplash_Thrown_Land_03}}
 {id=WavePlayer_2, properties {soundWaveAssetPtr=&GoldSplash_Thrown_Land_02}}
 {id=WavePlayer_3, properties {soundWaveAssetPtr=&GoldSplash_Thrown_Land_03}}
 {id=WavePlayer_4, properties {soundWaveAssetPtr=/VividRazorConsumablesGameplay/Assets/Audio/Waves/GoldSplash_Thrown_Land_04}}
 {id=WavePlayer_5, properties {soundWaveAssetPtr=/VividRazorConsumablesGameplay/Assets/Audio/Waves/GoldSplash_Thrown_Land_05}}
 {id=WavePlayer_6, properties {soundWaveAssetPtr=/VividRazorConsumablesGameplay/Assets/Audio/Waves/GoldSplash_Thrown_Land_06}}
 {id=WavePlayer_7, properties {soundWaveAssetPtr=/VividRazorConsumablesGameplay/Assets/Audio/Waves/GoldSplash_Thrown_Land_07}}
 {id=WavePlayer_8, properties {soundWaveAssetPtr=/VividRazorConsumablesGameplay/Assets/Audio/Waves/GoldSplash_Thrown_Land_08}}
 {id=WavePlayer_9, properties {soundWaveAssetPtr=&GoldSplash_Thrown_Land_01}}
]
properties {
 attenuationSettings=/Game/Sounds/Attenuation/Fort_Grenade_Smoke_Att
 bOverrideConcurrency=true
 maxDistance=6800
 soundClassObject=&Fort_Gadgets_Items
 volumeMultiplier=0.6
 attenuationOverrides {falloffDistance=2400, attenuationShapeExtents {x=600}}
 concurrencyOverrides {resolutionRule=PreventNew}
}
